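# Niagara effect for the CoffeeBreak emote's espresso pull, timing four emitters against sockets or bones on a shared Coffee skeletal mesh.
# EspressoPour draws a falling stream with two ribbon renderers and curve-offset Pos1/Pos2 positions.
# An indicator sprite switches socket and turns red to green midway, using age-driven curves and an array lookup.
# Later, a delayed coffee mesh burst and rising, gravity-affected ribbon trail appear and fade out.
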NiagaraSystem /BRCosmetics/Animation/Game/MainPlayer/Emotes/CoffeeBreak/FX/NS_Emote_CoffeeBreak_EspressoPull
dependencies {
 usedDataInterfaces [
  &NiagaraDataInterfaceColorCurve
  &NiagaraDataInterfaceCurve
  &NiagaraDataInterfaceVectorCurve
  &NiagaraDataInterfaceArrayInt32
  &NiagaraDataInterfaceSkeletalMesh
 ]
 usedDynamicInputs []
 usedModules [
  &SystemState
  &self:SystemSpawnScript.SystemScriptSource.SystemScriptGraph.NiagaraNodeAssignment_0.SetVariables_#1
 ]
}
emitters {
 CoffeeMesh {
  emitterUpdateScript [
   &EmitterState(Life Cycle Mode=::NewEnumerator1, Loop Behavior=::NewEnumerator1, Loop Delay=9.2, Loop Duration=0.908, UseLoopDelay=true)
   &SpawnBurst_Instantaneous(Use Spawn Probability=false, UseLoopCountLimit=false)
  ]
  particleSpawnScript [
   &InitializeParticle(
    Color=<rgba 0.1562 0.0417 0.0122 1>
    Lifetime=0.75
    Mesh Scale Mode=::NewEnumerator3
    Mesh Uniform Scale=0.32
    UsePositionOffset=false
    Write Material Random=true
   )
   &InitialMeshOrientation(Mesh Orientation Mode=::NewEnumerator6, Orientation Quaternion:Quat4f=<0 0 -3.1659 1>, Use Rotation Vector=false)
  ]
  particleUpdateScript [
   &SkeletalMeshLocation(
    Bone / Socket Index=5
    MaskSpawnGroup=false
    Mesh Sampling Type=::NewEnumerator5
    Sampled Position Offset=<2.0025 7.9537 -0.3177>
    Skeletal Mesh=$System.Coffee
    Socket Sampling Mode=::NewEnumerator5
   )
   &SolveForcesAndVelocity(Clamp Velocity=false, Limit Acceleration=false)
   &ParticleState()
   &DynamicMaterialParameters(
    Index 0 Param 2=-17.3228
    Index 0 Param 3=&FloatFromCurve(FloatCurve=DataInterface() {propertyValues {Curve {keys [<at 0 0.7779>,<at 1 2>]}}})
    Module.Param0WriteEnabled=true
    Module.Param0WriteEnabled1=true
    Module.Param1WriteEnabled=true
    Module.Param1WriteEnabled1=true
    Module.Param2WriteEnabled=true
    Module.Param2WriteEnabled1=true
    Module.Param3WriteEnabled=true
    Module.Param3WriteEnabled1=true
    Write Parameter Index 1=true
   )
   &ScaleColor(
    Scale Alpha=&FloatFromCurve(FloatCurve=DataInterface() {propertyValues {Curve {keys [<at 0.0339 -0.0471>,<at 0.0671 0.9559>,<at 0.4207 0.9559>,<at 0.76 0>]}}})
    ScaleA=true
    ScaleRGB=true
   )
  ]
  rendererClasses [&NiagaraMeshRendererProperties]
  renderers [{rendererClass=&NiagaraMeshRendererProperties, rendererIndex=0}]
 }
 Coffee_Trail {
  emitterUpdateScript [
   &EmitterState(Life Cycle Mode=::NewEnumerator1, Loop Behavior=::NewEnumerator1, Loop Delay=9.2, Loop Duration=0.36, UseLoopDelay=true)
   &SpawnRate(SpawnRate=100, Use Spawn Probability=false)
  ]
  particleSpawnScript [
   /SetVariables(Particles.VisibilityTag=&SpawnGroup()) {isSetParameters=true}
   &InitializeParticle(
    Color=<rgba 0.1562 0.0417 0.0122 1>
    Lifetime=0.4
    Mesh Scale Mode=::NewEnumerator3
    Position Offset Coordinate Space=::Local
    Ribbon Width=20
    Ribbon Width Mode=::NewEnumerator1
    Sprite Size Mode=::NewEnumerator3
    Uniform Sprite Size=2
    UsePositionOffset=true
    Write Material Random=true
   )
   &SkeletalMeshLocation(
    Bone / Socket Index=5
    MaskSpawnGroup=false
    Mesh Sampling Type=::NewEnumerator5
    Sampled Position Coordinate Space=::World
    Sampled Position Offset=<0 -8 4>
    Sampled Position Offset Coordinate Space:ENiagaraSkelMeshTransforms=::NewEnumerator2
    Skeletal Mesh=$System.Coffee
    Socket Sampling Mode=::NewEnumerator5
   )
   &AddVelocity(Velocity=<0 0 100>)
   &CameraOffset(Camera Offset Amount=5)
   &ShapeLocation(MaskSpawnGroup=true, Override Seed=false, Spawn Group Mask=1, Sphere Radius=46.0274, UseOwnerScale=true)
  ]
  particleUpdateScript [
   &GravityForce()
   &InheritSourceMovement(ApplyForce=false, ApplyPosition=true, ApplyPositionFalloff=false)
   &SolveForcesAndVelocity(Clamp Velocity=false, Limit Acceleration=false)
   &ParticleState()
   &ScaleRibbonWidth(Ribbon Width Scale=&FloatFromCurve(FloatCurve=DataInterface() {propertyValues {Curve {keys [<at 0 0.9 RCIM_Cubic>,<at 1 0.8 RCIM_Cubic>]}}}, Scale Curve=1.002))
   &DynamicMaterialParameters(Module.Param0WriteEnabled=true, Module.Param1WriteEnabled=true, Module.Param2WriteEnabled=true, Module.Param3WriteEnabled=true)
   &ScaleColor(
    Scale Alpha=&Multiply_Float(
     A=&FloatFromCurve(FloatCurve=DataInterface() {propertyValues {Curve {keys [<at -0.01 0.0353>,<at 0.0463 1>,<at 1 0>]}}})
     B=&FloatFromCurve(
      FloatCurve=DataInterface() {propertyValues {Curve {defaultValue=1, keys [{tangentMode=RCTM_User, time=9.4, value=1}, {tangentMode=RCTM_User, time=9.76, value=0}]}}}
      CurveIndex=$Emitter.Age
     )
    )
    ScaleA=true
    ScaleRGB=true
   )
  ]
  rendererClasses [&NiagaraRibbonRendererProperties]
  renderers [{rendererClass=&NiagaraRibbonRendererProperties}]
 }
 EspressoPour {
  emitterUpdateScript [
   &EmitterState(Life Cycle Mode=::NewEnumerator1, Loop Behavior=::NewEnumerator1, Loop Delay=1, Loop Duration=3.8, UseLoopDelay=true)
   &SpawnRate(SpawnRate=200, Use Spawn Probability=false)
  ]
  particleSpawnScript [
   &InitializeParticle(
    Color=<rgba 0.1406 0.0484 0.0147 1>
    Lifetime=0.3
    Position Mode=::NewEnumerator0
    Ribbon Width=0.7093
    Ribbon Width Mode=::NewEnumerator1
    Write Material Random=true
   )
   &AddVelocity(Velocity=<0 0 -39.655>)
  ]
  particleUpdateScript [
   &ParticleState()
   &SolveForcesAndVelocity(Clamp Velocity=false, Limit Acceleration=false)
   &ScaleColor(
    Scale Alpha=&FloatFromCurve(
     FloatCurve=DataInterface() {propertyValues {Curve {keys [
         {interpMode=RCIM_Cubic, leaveTangentWeight=1.4087, tangentMode=RCTM_User, time=0, value=0}
         <at 0.2 1 RCIM_Cubic>
         {arriveTangent=-4.5445, interpMode=RCIM_Cubic, leaveTangent=-4.5445, tangentMode=RCTM_User, time=1, value=0}
        ]}}}
     Scale Curve=2.5
    )
    Scale RGB=&MakeVectorFromLinearColorRGB(Color=&ColorFromCurve(ColorCurve=DataInterface() {propertyValues {
        AlphaCurve {keys []}
        BlueCurve {keys [<at 0.35 0.375 RCIM_Cubic>]}
        GreenCurve {keys [<at 0.35 0.375 RCIM_Cubic>]}
        RedCurve {keys [<at 0.35 0.375 RCIM_Cubic>]}
       }}))
    ScaleA=true
    ScaleRGB=true
   )
   /SetVariables(Particles.Pos1:NiagaraPosition=&AddVectorToPosition(
     Position=$Particles.Position
     Vector=&AddVectorToPosition(
      Position=<1.1595 0 0>
      Vector=&VectorFromCurve(
       VectorCurve=DataInterface() {propertyValues {
         XCurve {keys [
           {interpMode=RCIM_Cubic, leaveTangentWeight=0.1044, tangentMode=RCTM_Break, time=0, value=1}
           {
            arriveTangent=-0.4719
            arriveTangentWeight=0.0269
            interpMode=RCIM_Cubic
            leaveTangentWeight=0.0775
            tangentMode=RCTM_Break
            time=0.0153
            value=0.4086
           }
           <at 0.1744 1.1675>
          ]}
         YCurve {keys [
           {interpMode=RCIM_Cubic, tangentMode=RCTM_Break, time=0, value=1}
           {arriveTangent=-0.4719, interpMode=RCIM_Cubic, tangentMode=RCTM_Break, time=0.0153, value=0.4086}
           <at 0.1744 1.1675>
          ]}
         ZCurve {keys [
           {interpMode=RCIM_Cubic, leaveTangentWeight=0.1044, tangentMode=RCTM_Break, time=0, value=1}
           {
            arriveTangent=-0.4719
            arriveTangentWeight=0.0269
            interpMode=RCIM_Cubic
            leaveTangentWeight=0.0775
            tangentMode=RCTM_Break
            time=0.0153
            value=0.4086
           }
           <at 0.1744 1.1675>
          ]}
        }}
       Scale Curve=<-1.1582 0.3297 0>
      )
     )
    )) {isSetParameters=true}
   /SetVariables(Particles.Pos2:NiagaraPosition=&AddVectorToPosition(
     Position=$Particles.Position
     Vector=&AddVectorToPosition(
      Position=<-4.8666 0 0>
      Vector=&VectorFromCurve(
       VectorCurve=DataInterface() {propertyValues {
         XCurve {keys [
           {interpMode=RCIM_Cubic, leaveTangentWeight=0.1044, tangentMode=RCTM_Break, time=0, value=1}
           {
            arriveTangent=-0.4719
            arriveTangentWeight=0.0269
            interpMode=RCIM_Cubic
            leaveTangentWeight=0.0775
            tangentMode=RCTM_Break
            time=0.0153
            value=0.4086
           }
           <at 0.1744 1.1675>
          ]}
         YCurve {keys [
           {interpMode=RCIM_Cubic, tangentMode=RCTM_Break, time=0, value=1}
           {arriveTangent=-0.4719, interpMode=RCIM_Cubic, tangentMode=RCTM_Break, time=0.0153, value=0.4086}
           <at 0.1744 1.1675>
          ]}
         ZCurve {keys [
           {interpMode=RCIM_Cubic, leaveTangentWeight=0.1044, tangentMode=RCTM_Break, time=0, value=1}
           {
            arriveTangent=-0.4719
            arriveTangentWeight=0.0269
            interpMode=RCIM_Cubic
            leaveTangentWeight=0.0775
            tangentMode=RCTM_Break
            time=0.0153
            value=0.4086
           }
           <at 0.1744 1.1675>
          ]}
        }}
       Scale Curve=<1.1582 0.3297 0>
      )
     )
    )) {isSetParameters=true}
   &ScaleRibbonWidth(Ribbon Width Scale=&FloatFromCurve(FloatCurve=DataInterface() {propertyValues {Curve {keys [<at 0 1>, {arriveTangent=0.1792, interpMode=RCIM_Cubic, leaveTangent=-0.9861, tangentMode=RCTM_Break, time=0.125, value=0.8447}, <at 1 0.5>]}}}))
  ]
  rendererClasses [&NiagaraRibbonRendererProperties]
  renderers [
   {rendererClass=&NiagaraRibbonRendererProperties}
   {rendererClass=&NiagaraRibbonRendererProperties, rendererIndex=1}
  ]
 }
 light_innit {
  emitterUpdateScript [
   &EmitterState(Life Cycle Mode=::NewEnumerator1, Loop Behavior=::NewEnumerator1, UseLoopDelay=false)
   &SpawnBurst_Instantaneous(Use Spawn Probability=false, UseLoopCountLimit=false)
  ]
  particleSpawnScript [
   &InitializeParticle(
    Color=<rgba 0.9271 0.9271 0.9271 1>
    Lifetime=9.3
    Position Offset=<5 50 50>
    Position Offset Coordinate Space=::Local
    Ribbon Width=0.7093
    Ribbon Width Mode=::NewEnumerator1
    Sprite Size Mode=::NewEnumerator3
    Uniform Sprite Size=2.1361
    UsePositionOffset=true
    Write Material Random=true
   )
   &CameraOffset(Camera Offset Amount=5)
  ]
  particleUpdateScript [
   &ParticleState()
   &SkeletalMeshLocation(
    Bone / Socket Index=&SelectIntFromArray(
     Direct Array Index=&MakeIntFromFloat(Float=&FloatFromCurve(FloatCurve=DataInterface() {propertyValues {Curve {keys [<at 0 1 RCIM_Constant>,<at 4.5 0>]}}}, CurveIndex=$Particles.Age))
     Int Selection Array=DataInterface() {propertyValues {IntData [3,2]}}
    )
    MaskSpawnGroup=false
    Mesh Sampling Type=::NewEnumerator5
    Skeletal Mesh=$System.Coffee
    Socket Sampling Mode=::NewEnumerator5
   )
   &ScaleColor(
    Curve Index=$Particles.Age
    Linear Color Curve=DataInterface() {propertyValues {
      AlphaCurve {keys [<at 0 1>]}
      BlueCurve {keys [<at 0.4205 0 RCIM_Constant>,<at 0.5495 0 RCIM_Constant>]}
      GreenCurve {keys [<at 0.4205 0.0484 RCIM_Constant>,<at 4.5 2 RCIM_Constant>]}
      RedCurve {keys [<at 0.4205 2 RCIM_Constant>,<at 4.5 0.3116 RCIM_Constant>]}
     }}
    Scale Mode=2
   )
  ]
  rendererClasses [&NiagaraSpriteRendererProperties]
  renderers [{rendererClass=&NiagaraSpriteRendererProperties}]
 }
}
summary {
 systemName=NS_Emote_CoffeeBreak_EspressoPull
 emitters [
  {emitterName=EspressoPour, rendererClasses [&NiagaraRibbonRendererProperties]}
  {emitterName=light_innit, rendererClasses [&NiagaraSpriteRendererProperties]}
  {emitterName=Coffee_Trail, rendererClasses [&NiagaraRibbonRendererProperties]}
  {emitterName=CoffeeMesh, rendererClasses [&NiagaraMeshRendererProperties]}
 ]
 userVariables []
}
