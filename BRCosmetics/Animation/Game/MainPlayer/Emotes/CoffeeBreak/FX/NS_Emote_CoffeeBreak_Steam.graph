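# Niagara system for steam in the CoffeeBreak emote of the main player, built around a single emitter named Steam.
# The system stack uses SystemState and two Set Variables assignments, with no user variables.
# Its dependencies include a time-of-day color dynamic input and a skeletal mesh data interface, though their wiring is not shown.
# The emitter lists no renderers.
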NiagaraSystem /BRCosmetics/Animation/Game/MainPlayer/Emotes/CoffeeBreak/FX/NS_Emote_CoffeeBreak_Steam
dependencies {
 usedDataInterfaces [&NiagaraDataInterfaceSkeletalMesh]
 usedDynamicInputs [&FNTimeOfDayColor]
 usedModules [
  &self:SystemSpawnScript.SystemScriptSource.SystemScriptGraph.NiagaraNodeAssignment_1.SetVariables_#1
  &SystemState
  &self:SystemSpawnScript.SystemScriptSource.SystemScriptGraph.NiagaraNodeAssignment_0.SetVariables_#2
 ]
}
emitters {Steam {rendererClasses [], renderers []}}
summary {systemName=NS_Emote_CoffeeBreak_Steam, emitters [{emitterName=Steam, rendererClasses []}], userVariables []}
